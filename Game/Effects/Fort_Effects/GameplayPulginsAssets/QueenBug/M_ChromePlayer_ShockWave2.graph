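# Translucent, unlit, two-sided material for Niagara mesh particles, named as a shockwave effect for the QueenBug chrome player.
# Opacity combines particle alpha, a scale and a powered sine falloff across UV U, softened by DepthFade.
# SplineThicken takes one Width parameter for base and tip to widen thin mesh strips.
# Particle color feeds both an emissive adjustment function and a U-driven HueShift.
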Material /Game/Effects/Fort_Effects/GameplayPulginsAssets/QueenBug/M_ChromePlayer_ShockWave2
graph {
 properties {
  bUsedWithNiagaraMeshParticles=true
  blendMode=BLEND_Translucent
  pixelDepthOffsetMode=PDOM_Legacy
  shadingModel=MSM_Unlit
  translucencyPass=MTP_BeforeDOF
  twoSided=true
  usedShadingModels=Unlit
 }
 nodes [
  n1 = ScalarParameter() {defaultValue=0.12, parameterName=OpacityScale}
  n2 = MaterialFunctionCall(Float2=TextureCoordinate()) {materialFunction=&BreakOutFloat2Components}
  n3 = Power(Base=Sine(None=Multiply(A=@n2.R) {constB=0.5}))
  n4 = ParticleColor()
  DepthFade(Opacity=Multiply(A=Multiply(A=@n1, B=@n3), B=@n4.A), FadeDistance=ScalarParameter() {defaultValue=100, parameterName=DF})
  MaterialFunctionCall(Emissive=@n4.RGB) {materialFunction=&MF_FX_EmissiveAdjustments}
  n5 = ScalarParameter() {defaultValue=400, parameterName=Width}
  MaterialFunctionCall(WidthBase=@n5, WidthTip=@n5) {materialFunction=&SplineThicken}
  MaterialFunctionCall(Hue Shift Percentage=Multiply(A=@n2.R, B=ScalarParameter() {defaultValue=2, parameterName=HS}), Texture=@n4.RGB) {materialFunction=&HueShift}
 ]
}
